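# A demo material named black, stored with the drivable Hoverboard vehicle assets.
# It holds two unnamed-in-file vector parameters, one defaulting to near-black and one to white, with no textures or math nodes and no visible output wiring.
# It uses Default Lit shading and is flagged for skeletal meshes, morph targets and mesh deformers.
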Material /Game/Athena/DrivableVehicles/Hoverboard/DEMO/black
graph {
 properties {
  bUsedWithMeshDeformer=true
  bUsedWithMorphTargets=true
  bUsedWithSkeletalMesh=true
  pixelDepthOffsetMode=PDOM_Legacy
  usedShadingModels=Default Lit
 }
 nodes [
  VectorParameter() {defaultValue=<rgba 0.0083 0.01 0.01 0>}
  VectorParameter() {defaultValue=<rgba 1 1 1 0>}
 ]
}
